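# Physics asset for the HeatSnake Drift rifle skeletal mesh, providing its simple physics collision shape.
# It holds a single body on the handle bone, a capsule offset along the bone and rotated to lie sideways relative to it.
# There are no constraints, so the rifle acts as one rigid body with no articulated parts.
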PhysicsAsset /HeatSnake_Drift/Assets/HeatSnake_Drift_Rifle/Meshes/Heatsnake_Drift_Rifle_PhysicsAsset
bodies {handle [{shapeType=Capsule, length=40.1621, radius=14.9971, center=<13.415 -0.0021 3.575>, rotation=<rot 0 90.0166 -89.807>}]}
constraints []
